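# Physics asset for the DoubleDutyStew backpack cosmetic, defining collision and joint limits for its cloth bones.
# A small capsule sits on cloth_master_jnt, and a larger zero-length capsule, effectively a sphere, sits on dyn_cloth_jnt.
# A single constraint links the two bodies with narrow swing and twist limits, which restricts how far the dynamic cloth bone can rotate relative to the master joint.
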PhysicsAsset /BRCosmetics/Accessories/FORT_Backpacks/Backpack_M_MED_DoubleDutyStew/Meshes/M_MED_DoubleDutyStew_Pack_Physics
bodies {
 cloth_master_jnt [{shapeType=Capsule, length=0.505, radius=0.505, shapeName=cloth_master_jnt_capsule, rotation=<rot 0 90 0>}]
 dyn_cloth_jnt [{shapeType=Capsule, length=0, radius=2.6555, shapeName=dyn_cloth_jnt_capsule}]
}
constraints [{bone1Name=dyn_cloth_jnt, bone2Name=cloth_master_jnt, swing1LimitDegrees=10, swing2LimitDegrees=10, twistLimitDegrees=15}]
