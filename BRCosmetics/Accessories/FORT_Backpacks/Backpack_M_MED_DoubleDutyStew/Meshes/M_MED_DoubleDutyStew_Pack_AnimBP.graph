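# Animation Blueprint for the DoubleDutyStew backpack cosmetic, adding physics and state-based posing to the pack.
# It copies the mesh pose into two RigidBody simulations blended by front-end state, then chains additive offsets on the pack bone.
# Each update derives gravity, acceleration scale and offset toggles from player states like jumping, skydiving, gliding, melee and emotes.
# Spine and clavicle rotations drive some offsets; cloth and wind settings are unused in these graphs.
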AnimBlueprint /BRCosmetics/Accessories/FORT_Backpacks/Backpack_M_MED_DoubleDutyStew/Meshes/M_MED_DoubleDutyStew_Pack_AnimBP
target_skeleton=&/M_MED_DoubleDutyStew_Pack_Skeleton
graphs [
 AnimGraph {nodes [
   n1 = Get EmoteAndMeleeAnim_cond() -> (EmoteAndMeleeAnim_cond:bool)
   n2 = "Transform (Modify) Bone - Bone: pack"!(
    ComponentPose=Local To Component(LocalPose=Use cached pose 'dyn_cache'())
    bAlphaBoolEnabled:bool=@n1
    Translation="0,0,0"
    Rotation=Get rawRotationSpine() -> (rawRotationSpine:Rotator)
    Scale
   ) -> (Pose) {
    alphaInputType=Bool
    rotationMode=BMM_Additive
    translationMode=BMM_Additive
    alphaBoolBlend {blendInTime=0.1, blendOption=Linear, blendOutTime=0.16}
    boneToModify {boneName=pack}
    rotation {roll=20}
   }
   n3 = "Transform (Modify) Bone - Bone: pack"(ComponentPose=@n2, bAlphaBoolEnabled=@n1, Translation="0,0,0", Rotation=Get Return Value Rotation() -> (Return Value Rotation:Rotator)) {
    alphaInputType=Bool
    rotationMode=BMM_Additive
    translationMode=BMM_Additive
    alphaBoolBlend {blendInTime=0.1, blendOption=Linear, blendOutTime=0.16}
    boneToModify {boneName=pack}
   }
   n4 = "Transform (Modify) Bone - Bone: pack"(ComponentPose=@n3, bAlphaBoolEnabled=Get bGreaterVel() -> (bGreaterVel:bool), Translation="0,-3,0", Rotation="0,0,20") {
    alphaInputType=Bool
    rotationMode=BMM_Additive
    translationMode=BMM_Additive
    alphaBoolBlend {blendInTime=0.1, blendOption=Linear, blendOutTime=0.5}
    alphaScaleBiasClamp {bInterpResult=true, interpSpeedDecreasing=0.5, interpSpeedIncreasing=5}
    boneToModify {boneName=pack}
    rotation {roll=20}
    translation {y=-3}
   }
   n5 = "Transform (Modify) Bone - Bone: pack"(ComponentPose=@n4, bAlphaBoolEnabled=Get NewVar() -> (NewVar:bool), Translation="0,2,0", Rotation="0,0,-1.5") {
    alphaInputType=Bool
    rotationMode=BMM_Additive
    translationMode=BMM_Additive
    alphaBoolBlend {blendInTime=10, blendOption=Linear, blendOutTime=0.3}
    alphaScaleBiasClamp {bInterpResult=true, interpSpeedDecreasing=0.5, interpSpeedIncreasing=5}
    boneToModify {boneName=pack}
    rotation {roll=-1.5}
    translation {y=2}
   }
   n6 = "Transform (Modify) Bone - Bone: pack"(ComponentPose=@n5, bAlphaBoolEnabled=Get bIsJumping(), Translation="0,0,0", Rotation="0,0,8") {
    alphaInputType=Bool
    rotationMode=BMM_Additive
    translationMode=BMM_Additive
    alphaBoolBlend {blendInTime=0.3, blendOption=Linear, blendOutTime=0.5}
    alphaScaleBiasClamp {bInterpResult=true, interpSpeedDecreasing=0.5, interpSpeedIncreasing=5}
    boneToModify {boneName=pack}
    rotation {roll=8}
   }
   n7 = "Transform (Modify) Bone - Bone: pack"!(
    ComponentPose=@n6
    bAlphaBoolEnabled="Get Fly_condition "() -> ("Fly_condition ":bool)
    Translation="0,0,0"
    Rotation_Roll:float=Get!() -> (RotateSpineCond_inSky_Roll:float, RotateSpineCond_inSky_Pitch:float, RotateSpineCond_inSky_Yaw:float).RotateSpineCond_inSky_Yaw
    Rotation_Pitch:float=0
    Rotation_Yaw:float=0
    Scale
   ) -> (Pose) {
    alphaInputType=Bool
    rotationMode=BMM_Additive
    translationMode=BMM_Additive
    alphaBoolBlend {blendInTime=1, blendOption=Linear, blendOutTime=0.4}
    alphaScaleBiasClamp {bInterpResult=true, interpSpeedDecreasing=0.5, interpSpeedIncreasing=5}
    boneToModify {boneName=pack}
    rotation {roll=10}
   }
   n8 = "Transform (Modify) Bone - Bone: pack"!(ComponentPose=@n7, bAlphaBoolEnabled, Translation="0,-1,0", Rotation:Rotator="0,0,8", Scale) -> (Pose) {
    alphaInputType=Bool
    rotationMode=BMM_Additive
    translationMode=BMM_Additive
    alphaBoolBlend {blendInTime=0.3, blendOption=Linear, blendOutTime=0.3}
    alphaScaleBiasClamp {bInterpResult=true, interpSpeedDecreasing=0.5, interpSpeedIncreasing=5}
    boneToModify {boneName=pack}
    rotation {roll=8}
    translation {y=-1}
   }
   n9 = "Transform (Modify) Bone - Bone: pack"!(
    ComponentPose=@n8
    bAlphaBoolEnabled=Get Para Glide Open Glider() -> (Para Glide Open Glider:bool)
    Translation="0,0,0"
    Rotation_Roll:float=Get!() -> (RotateSpineCond_inSky_max_Roll:float, RotateSpineCond_inSky_max_Pitch:float, RotateSpineCond_inSky_max_Yaw:float).RotateSpineCond_inSky_max_Yaw
    Rotation_Pitch:float=0
    Rotation_Yaw:float=0
    Scale
   ) -> (Pose) {
    alphaInputType=Bool
    rotationMode=BMM_Additive
    alphaBoolBlend {blendInTime=0.3, blendOption=Linear, blendOutTime=5}
    boneToModify {boneName=pack}
   }
   n10 = "Transform (Modify) Bone - Bone: pack"!(ComponentPose=@n9, bAlphaBoolEnabled, Translation="0,0,0", Rotation:Rotator="0,0,5", Scale) -> (Pose) {
    alphaInputType=Bool
    rotationMode=BMM_Additive
    translationMode=BMM_Additive
    alphaBoolBlend {blendOption=Linear}
    boneToModify {boneName=pack}
    rotation {roll=5}
   }
   n11 = "Transform (Modify) Bone - Bone: pack"(ComponentPose=@n10, Translation="0,0,-1.1", Rotation="0,0,0") {
    alphaInputType=Bool
    rotationMode=BMM_Additive
    translationMode=BMM_Additive
    alphaBoolBlend {blendInTime=10, blendOption=Linear, blendOutTime=0.3}
    alphaScaleBiasClamp {bInterpResult=true, interpSpeedDecreasing=0.5, interpSpeedIncreasing=5}
    boneToModify {boneName=pack}
    translation {z=-1.1}
   }
   Output Pose(Result=Component To Local(ComponentPose=@n11)) {layerGroup=None}
   n12 = Get Override World Gravity() -> (Override World Gravity:Vector)
   n13 = RigidBody!(
    ComponentPose=Local To Component(LocalPose=Copy Pose From Mesh())
    Alpha
    OverrideWorldGravity:Vector=@n12
    ExternalForce
    ComponentLinearAccScale:Vector=Get Component Linear Acc Scale() -> (Component Linear Acc Scale:Vector)
   ) -> (Pose) {
    bOverrideWorldGravity=true
    overridePhysicsAsset=&/M_MED_DoubleDutyStew_Pack_Physics
    alphaBoolBlend {blendOption=Linear}
    componentAppliedLinearAccClamp=<1000 1000 1000>
    componentLinearAccScale=<0.1 0.1 0.2>
    componentLinearVelScale=<0.2 0.2 0.2>
    overrideWorldGravity {z=-980}
    simSpaceSettings {maxAngularAcceleration=2000, maxAngularVelocity=150, maxLinearAcceleration=1300, maxLinearVelocity=10}
   }
   n14 = RigidBody(ComponentPose=Local To Component(LocalPose=Copy Pose From Mesh()), OverrideWorldGravity=@n12, ComponentLinearAccScale="0,0,0") {
    bOverrideWorldGravity=true
    overridePhysicsAsset=&M_MED_DoubleDutyStew_Pack_Physics
    alphaBoolBlend {blendOption=Linear}
    componentAppliedLinearAccClamp=<1000 1000 1000>
    componentLinearAccScale=<0.1 0.1 0.2>
    componentLinearVelScale=<0.04 0.04 0.04>
    overrideWorldGravity {z=-980}
    simSpaceSettings {maxAngularAcceleration=2000, maxAngularVelocity=150, maxLinearAcceleration=1300, maxLinearVelocity=10}
   }
   n15 = Blend Poses by bool(BlendPose_0=Component To Local(ComponentPose=@n14), BlendPose_1=Component To Local(ComponentPose=@n13), bActiveValue=Get bIsInFrontEnd()) {blendType=Linear}
   Save cached pose 'dyn_cache'(Pose=Component To Local(ComponentPose=Local To Component(LocalPose=@n15))) {cacheName=dyn_cache}
  ]}
 EventGraph {nodes [
   Event BlueprintUpdateAnimation()
   > n1 = Sequence!(execute) -> (then_0, then_1, then_2:Exec, then_3:Exec, then_4:Exec)
   n2 = Sequence!(execute=@n1.then_4) -> (then_0, then_1, then_2)
   n3 = "Min (Float)"(
    A:double=GetAlphaFromReferenceTransformDelta(SkeletalMeshComponent=GetOwnerSkeletalMeshComponent(), BoneName=spine_04, BaseName=spine_03, InputRange=(X=0,Y=1), TwistAxis="0,0,1").RawOutput
    B:double=0
   ) -> (ReturnValue:double) {is=title K2Node_CommutativeAssociativeBinaryOperator}
   n4 = Invert Rotator!(A:Rotator=MakeRotator(Yaw=@n3)) -> (ReturnValue:Rotator)
   n5 = BreakRotator(InRot=@n4)
   > Set RotateSpineCond_inSky(RotateSpineCond_inSky:Rotator=MakeRotator(Yaw=float * float(A=@n5.Yaw, B=4) {is=title K2Node_PromotableOperator})) -> (Output_Get:Rotator)
   > "Set Fly_condition "("Fly_condition ":bool=AND Boolean(A=Get bIsSkydiveFloating(), B="float < float"(A=@n3, B=-15))) -> (Output_Get:bool)
   Set Skel_Mesh(execute=@n2.then_2, Skel_Mesh=GetOwnerSkeletalMeshComponent())
   n6 = GetOwnerFortPlayerAnimInstance!(self:Custom Character Part Anim Instance Object Reference) -> (ReturnValue:Fort Player Anim Instance Object Reference)
   > Set Para Glide Open Glider(Para Glide Open Glider:bool=Get bParaGlide_OpenGlider!(self:Fort Player Anim Instance Object Reference=@n6) -> (bParaGlide_OpenGlider:bool)) -> (Output_Get:bool)
   > Set Is Playing Emote(Is Playing Emote:bool=Get bIsPlayingEmote!(self:Fort Player Anim Instance Object Reference=@n6) -> (bIsPlayingEmote)) -> (Output_Get:bool)
   n7 = BreakRotator(InRot=@n4)
   Set RotateSpineCond_inSky_max(execute=@n2.then_1, RotateSpineCond_inSky_max:Rotator=MakeRotator(Yaw=float * float(A=@n7.Yaw, B=5))) -> (Output_Get:Rotator)
   n8 = Get bIsPlayingMeleeAnim()
   n9 = AND Boolean(A=AND Boolean(A=Get bIsOnGround() -> (bIsOnGround:bool), B=NOT Boolean(A=@n8)), B=NOT Boolean(A=Get bIsJumping()))
   Set NewVar(execute=@n1.then_0, NewVar:bool=@n9) -> (Output_Get:bool)
   n10 = Sequence!(execute=@n1.then_2) -> (then_0, then_1)
   n11 = "float > float"(A="Max (Float)"(A=Get PawnVelocityZ(), B=0), B=3000)
   n12 = "Not Equal (Boolean)"(A=Get bIsParachuteOpened(), B=true) -> (ReturnValue=~)
   > n13 = Branch(Condition=AND Boolean(A=AND Boolean(A=@n11, B=@n12), B=Get bIsSkyDiving()))
   > Set bGreaterVel(bGreaterVel:bool=true) -> (Output_Get:bool)
   Set bGreaterVel(execute=@n13.else, bGreaterVel=false)
   n14 = Branch(execute=@n10.then_1, Condition=Get bIsJumping())
   > n15 = Set Override World Gravity(Override World Gravity:Vector="0,0,-250") -> (Output_Get:Vector)
   Set Component Linear Acc Scale(execute=[@n17.then, @n15.then], Component Linear Acc Scale:Vector="0.1,0.1,0.2") -> (Output_Get:Vector)
   n16 = Branch(execute=@n14.else, Condition=Get bIsInFrontEnd())
   > Set Override World Gravity(Override World Gravity="0,0,-200")
   > Set Component Linear Acc Scale(Component Linear Acc Scale="0.012,0.012,0.012")
   n17 = Set Override World Gravity(execute=@n16.else, Override World Gravity="0,0,-350")
   n18 = Branch(execute=@n1.then_3, Condition=Get bIsParachuteOpened())
   > Set RawRotateSpine_InSky(RawRotateSpine_InSky:float=0) -> (Output_Get:float)
   n19 = Branch(execute=@n18.else, Condition=Get bIsSkyDiving())
   > Set RawRotateSpine_InSky(RawRotateSpine_InSky=1)
   n20 = Set EmoteAndMeleeAnim_cond(execute=@n19.else, EmoteAndMeleeAnim_cond:bool=OR Boolean(A=Get bIsPlayingMeleeAnim(), B=Get Is Playing Emote() -> (Is Playing Emote:bool))) -> (Output_Get:bool)
   > Branch(Condition=@n20)
   n21 = GetOwnerSkeletalMeshComponent()
   n22 = float - float(B=MakeRotator!(
     Roll
     Pitch
     Yaw=GetAlphaFromReferenceTransformDelta(SkeletalMeshComponent=@n21, BoneName=spine_05, BaseName=spine_04, InputRange=(X=0,Y=1), TwistAxis="0,0,1").RawOutput
    ) -> (ReturnValue_Roll:float, ReturnValue_Pitch:float, ReturnValue_Yaw:float).ReturnValue_Yaw)
   > Set rawRotationSpine(rawRotationSpine:Rotator=MakeRotator!(Roll=float / float(A=@n22, B=1.2), Pitch, Yaw) -> (ReturnValue:Rotator)) -> (Output_Get:Rotator)
   n23 = GetOwnerSkeletalMeshComponent()
   n24 = "Min (Float)"(
    A=BreakRotator(InRot=GetSocketTransform!(self=@n23, InSocketName=clavicle_l, TransformSpace=RTS_Component) -> (ReturnValue_Location:Vector, ReturnValue_Rotation:Rotator, ReturnValue_Scale:Vector).ReturnValue_Rotation).Yaw
    B=0
   )
   > Set Return Value Rotation(Return Value Rotation:Rotator=MakeRotator(Yaw=float / float(A=@n24, B=4))) -> (Output_Get:Rotator)
   TryGetPawnOwner()
   AND Boolean(A=@n9, B=NOT Boolean(A=Get bIsCrouching()))
  ]}
]
variables [
 " Gust Local Settings" {type=WindGustLocalSettingStruct, default {elapsedTime=0, finalGustStrengthLocal=0, gustDuration=0, gustStrength=0, timeBetweenGusts=0}}
 bGreaterVel {type=boolean}
 Cloth Core Settings {
  type=ClothCoreSettingStruct
  default {
   collisionThickness=1.5
   dampingCoefficient=0.1
   frictionCoefficient=0.5
   localDampingCoefficient=0.1
   numberIterations=2
   numberSubsteps=1
   selfCollisionThickness=0
   animDriveDamping=<0 1>
   animDriveStiffness=<0 1>
   areaStiffness=<1 1>
   bendingStiffness=<1 1>
   edgeStiffness=<1 1>
   tetherStiffness=<1 1>
  }
 }
 Cloth Paramaters {
  type=ClothParamsStruct
  default {
   dBNO {
    angularVelocityScale=0.6
    fictitiousAngularScale=0.1
    jointName=pelvis
    dragCoefficient=<0.2 1>
    gravityOverride=<-980 400 0>
    liftCoefficient=<0 0.5>
    linearVelocityScale=<0.3 0.3 0.1>
   }
   dBNO_Carried {
    angularVelocityScale=0.6
    fictitiousAngularScale=0.1
    jointName=pelvis
    dragCoefficient=<0.2 1>
    gravityOverride=<-980 -200 0>
    liftCoefficient=<0 0.5>
    linearVelocityScale=<0.3 0.3 0.1>
   }
   dBNO_Impulse {
    angularVelocityScale=0
    fictitiousAngularScale=0
    jointName=pelvis
    dragCoefficient=<0 0.2>
    gravityOverride=<-3000 -2000 0>
    liftCoefficient=<0 0.1>
    linearVelocityScale=<0.3 0.1 0.3>
   }
   dBNO_RollBwd {
    angularVelocityScale=0
    fictitiousAngularScale=0
    jointName=pelvis
    dragCoefficient=<0 0.2>
    gravityOverride=<-2000 0 0>
    liftCoefficient=<0 0.1>
    linearVelocityScale=<0.3 0.1 0.3>
   }
   dBNO_RollFwd {
    angularVelocityScale=0
    fictitiousAngularScale=0
    jointName=pelvis
    dragCoefficient=<0 0.2>
    gravityOverride=<-980 0 0>
    liftCoefficient=<0 0.1>
    linearVelocityScale=<0.3 0.1 0.3>
   }
   dBNO_Tumbling {
    angularVelocityScale=0
    fictitiousAngularScale=0
    jointName=pelvis
    dragCoefficient=<0 0.2>
    gravityOverride=<-2500 -300 0>
    liftCoefficient=<0 0.1>
    linearVelocityScale=<0.3 0.1 0.3>
   }
   emoteOrMelee {
    angularVelocityScale=0.6
    fictitiousAngularScale=0.1
    jointName=root
    dragCoefficient=<2.5 2.5>
    gravityOverride=<0 0 -980>
    liftCoefficient=<0.5 0.5>
    linearVelocityScale=<0.6 0.6 0.5>
   }
   front_End {
    angularVelocityScale=0.05
    fictitiousAngularScale=0.6
    jointName=root
    dragCoefficient=<0.2 0.2>
    gravityOverride=<0 -200 -980>
    liftCoefficient=<0.1 0.1>
    linearVelocityScale=<0.1 0.1 0.1>
   }
   onGround_Crouch_Moving {
    angularVelocityScale=0.1
    fictitiousAngularScale=0.3
    jointName=root
    dragCoefficient=<2 2>
    gravityOverride=<0 0 -980>
    liftCoefficient=<0.5 0.5>
    linearVelocityScale=<0.6 0.6 0.5>
   }
   onGround_Falling {
    angularVelocityScale=0.6
    fictitiousAngularScale=0
    jointName=root
    dragCoefficient=<0 0.2>
    gravityOverride=<0 -500 -980>
    liftCoefficient=<0 0.1>
    linearVelocityScale=<2 1 0.3>
   }
   onGround_Moving {
    angularVelocityScale=0.6
    fictitiousAngularScale=0.1
    jointName=root
    dragCoefficient=<2 2>
    gravityOverride=<0 0 -980>
    liftCoefficient=<0.5 0.5>
    linearVelocityScale=<0.6 0.6 0.5>
   }
   onGround_Standing {
    angularVelocityScale=0.05
    fictitiousAngularScale=1.2
    jointName=root
    dragCoefficient=<2.5 2.5>
    gravityOverride=<0 0 -980>
    liftCoefficient=<0.5 0.5>
    linearVelocityScale=<0.6 0.6 0.5>
   }
   rail_Grinding {
    angularVelocityScale=0.2
    fictitiousAngularScale=0
    jointName=root
    dragCoefficient=<0.1 0.2>
    gravityOverride=<50 -500 -400>
    liftCoefficient=<0 0.5>
    linearVelocityScale=<0.1 0.1 0.1>
   }
   riding_Moving {
    angularVelocityScale=0.05
    fictitiousAngularScale=1
    jointName=root
    dragCoefficient=<0.5 0.5>
    gravityOverride=<0 0 -980>
    liftCoefficient=<0.5 0.5>
    linearVelocityScale=<0.4 0.6 0.3>
   }
   riding_Moving_Targeting {
    angularVelocityScale=0.05
    fictitiousAngularScale=0.7
    jointName=root
    dragCoefficient=<0.5 0.5>
    gravityOverride=<0 0 -980>
    liftCoefficient=<0.5 0.5>
    linearVelocityScale=<0.2 0.6 0.2>
   }
   riding_Standing {
    angularVelocityScale=0.05
    fictitiousAngularScale=1
    jointName=root
    dragCoefficient=<0.5 0.5>
    gravityOverride=<0 -450 -980>
    liftCoefficient=<0.5 0.5>
    linearVelocityScale=<0.6 0.6 0.3>
   }
   skydive_DivingDown {
    angularVelocityScale=0.1
    fictitiousAngularScale=0
    jointName=pelvis
    dragCoefficient=<0.3 0.3>
    gravityOverride=<-450 -100 0>
    liftCoefficient=<0.9 0.9>
    linearVelocityScale=<0.1 0.1 0.1>
   }
   skydive_FloatingDown {
    angularVelocityScale=0.1
    fictitiousAngularScale=0
    jointName=root
    dragCoefficient=<0.3 0.3>
    gravityOverride=<0 -800 200>
    liftCoefficient=<1.3 1.3>
    linearVelocityScale=<0.2 0.2 0.1>
   }
   skydive_GoingUp {
    angularVelocityScale=0.1
    fictitiousAngularScale=0
    jointName=root
    dragCoefficient=<0.01 0.01>
    gravityOverride=<0 -2000 400>
    liftCoefficient=<0.05 0.05>
    linearVelocityScale=<0.1 0.1 0.01>
   }
   skydive_Parachuting {
    angularVelocityScale=0.6
    fictitiousAngularScale=0.1
    jointName=root
    dragCoefficient=<0.8 0.8>
    gravityOverride=<0 -600 -500>
    liftCoefficient=<0.8 0.8>
    linearVelocityScale=<0.2 0.2 0.3>
   }
   skydive_Surfing {
    angularVelocityScale=0.6
    fictitiousAngularScale=0.1
    jointName=root
    dragCoefficient=<1.5 1.5>
    gravityOverride=<0 -600 -500>
    liftCoefficient=<1 1>
    linearVelocityScale=<0.1 0.1 0.3>
   }
   swimming {
    angularVelocityScale=0.6
    fictitiousAngularScale=0
    jointName=root
    dragCoefficient=<2.5 2.5>
    gravityOverride=<0 0 0>
    liftCoefficient=<0.1 0.1>
    linearVelocityScale=<0.6 0.6 0.5>
   }
   tactical_Sprint {
    angularVelocityScale=0.6
    fictitiousAngularScale=0.1
    jointName=root
    dragCoefficient=<2 2>
    gravityOverride=<0 0 -980>
    liftCoefficient=<0.5 0.5>
    linearVelocityScale=<0.6 0.6 0.5>
   }
   ziplining {
    angularVelocityScale=0.6
    fictitiousAngularScale=0
    jointName=root
    dragCoefficient=<0 1.5>
    gravityOverride=<0 0 -980>
    liftCoefficient=<0 0.3>
    linearVelocityScale=<0.4 0.4 0.5>
   }
  }
 }
 Component Linear Acc Scale {type=Vector, default=<0.1 0.1 0.2>}
 EmoteAndMeleeAnim_cond {type=boolean}
 "Fly_condition " {type=boolean}
 Is Playing Emote {type=boolean}
 NewVar {type=boolean}
 Override World Gravity {type=Vector, default=<0 0 -980>}
 Para Glide Open Glider {type=boolean}
 RawRotateSpine_InSky {type=number, default=0.01}
 rawRotationSpine {type=Rotator}
 RbanOrCloth {type=boolean, default=true}
 Return Value Rotation {type=Rotator}
 RotateSpineCond_inSky {type=Rotator}
 RotateSpineCond_inSky_max {type=Rotator}
 Skel_Mesh {type=SkeletalMeshComponent}
 Wind Gust Local Settings {type=WindGustLocalSettingStruct, default {elapsedTime=0, finalGustStrengthLocal=0, gustDuration=0, gustStrength=0, timeBetweenGusts=0}}
 Wind Gusts Params {
  type=WindGustParamsStruct
  default {
   dBNO {
    maxDownInterpSpeed=5
    maxGustDuration=2
    maxGustStrength=7
    maxTimeBetweenGusts=7
    maxUpInterpSpeed=5
    minDownInterpSpeed=2
    minGustDuration=0.5
    minGustStrength=3
    minTimeBetweenGusts=0
    minUpInterpSpeed=2
    windTurbulenceScalarA=0.1
    windTurbulenceScalarB=1
    windAmplitude=<0 0 0>
    windFrequency=<0 0 0>
    windOffset=<0 0 0>
   }
   dBNO_Carried {
    maxDownInterpSpeed=5
    maxGustDuration=2
    maxGustStrength=7
    maxTimeBetweenGusts=7
    maxUpInterpSpeed=5
    minDownInterpSpeed=2
    minGustDuration=0.5
    minGustStrength=3
    minTimeBetweenGusts=0
    minUpInterpSpeed=2
    windTurbulenceScalarA=0.1
    windTurbulenceScalarB=1
    windAmplitude=<0 0 0>
    windFrequency=<0 0 0>
    windOffset=<0 0 0>
   }
   dBNO_Impulse {
    maxDownInterpSpeed=5
    maxGustDuration=2
    maxGustStrength=7
    maxTimeBetweenGusts=7
    maxUpInterpSpeed=5
    minDownInterpSpeed=2
    minGustDuration=0.5
    minGustStrength=3
    minTimeBetweenGusts=0
    minUpInterpSpeed=2
    windTurbulenceScalarA=0.1
    windTurbulenceScalarB=1
    windAmplitude=<0 0 0>
    windFrequency=<0 0 0>
    windOffset=<0 0 0>
   }
   dBNO_RollBwd {
    maxDownInterpSpeed=5
    maxGustDuration=2
    maxGustStrength=7
    maxTimeBetweenGusts=7
    maxUpInterpSpeed=5
    minDownInterpSpeed=2
    minGustDuration=0.5
    minGustStrength=3
    minTimeBetweenGusts=0
    minUpInterpSpeed=2
    windTurbulenceScalarA=0.1
    windTurbulenceScalarB=1
    windAmplitude=<0 0 0>
    windFrequency=<0 0 0>
    windOffset=<0 0 0>
   }
   dBNO_RollFwd {
    maxDownInterpSpeed=5
    maxGustDuration=2
    maxGustStrength=7
    maxTimeBetweenGusts=7
    maxUpInterpSpeed=5
    minDownInterpSpeed=2
    minGustDuration=0.5
    minGustStrength=3
    minTimeBetweenGusts=0
    minUpInterpSpeed=2
    windTurbulenceScalarA=0.1
    windTurbulenceScalarB=1
    windAmplitude=<0 0 0>
    windFrequency=<0 0 0>
    windOffset=<0 0 0>
   }
   dBNO_Tumbling {
    maxDownInterpSpeed=5
    maxGustDuration=2
    maxGustStrength=7
    maxTimeBetweenGusts=7
    maxUpInterpSpeed=5
    minDownInterpSpeed=2
    minGustDuration=0.5
    minGustStrength=3
    minTimeBetweenGusts=0
    minUpInterpSpeed=2
    windTurbulenceScalarA=0.1
    windTurbulenceScalarB=1
    windAmplitude=<0 0 0>
    windFrequency=<0 0 0>
    windOffset=<0 0 0>
   }
   emoteOrMelee {
    maxDownInterpSpeed=5
    maxGustDuration=2
    maxGustStrength=7
    maxTimeBetweenGusts=7
    maxUpInterpSpeed=5
    minDownInterpSpeed=2
    minGustDuration=0.5
    minGustStrength=3
    minTimeBetweenGusts=0
    minUpInterpSpeed=2
    windTurbulenceScalarA=0.1
    windTurbulenceScalarB=1
    windAmplitude=<0 0 0>
    windFrequency=<0 0 0>
    windOffset=<0 0 0>
   }
   front_End {
    maxDownInterpSpeed=5
    maxGustDuration=2
    maxGustStrength=7
    maxTimeBetweenGusts=7
    maxUpInterpSpeed=5
    minDownInterpSpeed=2
    minGustDuration=0.5
    minGustStrength=3
    minTimeBetweenGusts=0
    minUpInterpSpeed=2
    windTurbulenceScalarA=0.1
    windTurbulenceScalarB=1
    windAmplitude=<0 0 0>
    windFrequency=<0 0 0>
    windOffset=<0 0 0>
   }
   onGround_Crouch_Moving {
    maxDownInterpSpeed=5
    maxGustDuration=2
    maxGustStrength=7
    maxTimeBetweenGusts=7
    maxUpInterpSpeed=5
    minDownInterpSpeed=2
    minGustDuration=0.5
    minGustStrength=3
    minTimeBetweenGusts=0
    minUpInterpSpeed=2
    windTurbulenceScalarA=0.1
    windTurbulenceScalarB=1
    windAmplitude=<0 0 0>
    windFrequency=<0 0 0>
    windOffset=<0 0 0>
   }
   onGround_Falling {
    maxDownInterpSpeed=5
    maxGustDuration=2
    maxGustStrength=7
    maxTimeBetweenGusts=7
    maxUpInterpSpeed=5
    minDownInterpSpeed=2
    minGustDuration=0.5
    minGustStrength=3
    minTimeBetweenGusts=0
    minUpInterpSpeed=2
    windTurbulenceScalarA=0.1
    windTurbulenceScalarB=1
    windAmplitude=<0 0 0>
    windFrequency=<0 0 0>
    windOffset=<0 0 0>
   }
   onGround_Moving {
    maxDownInterpSpeed=5
    maxGustDuration=2
    maxGustStrength=7
    maxTimeBetweenGusts=7
    maxUpInterpSpeed=5
    minDownInterpSpeed=2
    minGustDuration=0.5
    minGustStrength=3
    minTimeBetweenGusts=0
    minUpInterpSpeed=2
    windTurbulenceScalarA=0.1
    windTurbulenceScalarB=1
    windAmplitude=<0 0 0>
    windFrequency=<0 0 0>
    windOffset=<0 0 0>
   }
   onGround_Standing {
    maxDownInterpSpeed=5
    maxGustDuration=2
    maxGustStrength=7
    maxTimeBetweenGusts=7
    maxUpInterpSpeed=5
    minDownInterpSpeed=2
    minGustDuration=0.5
    minGustStrength=3
    minTimeBetweenGusts=0
    minUpInterpSpeed=2
    windTurbulenceScalarA=0.1
    windTurbulenceScalarB=1
    windAmplitude=<0 0 0>
    windFrequency=<0 0 0>
    windOffset=<0 0 0>
   }
   rail_Grinding {
    maxDownInterpSpeed=5
    maxGustDuration=8
    maxGustStrength=7
    maxTimeBetweenGusts=8
    maxUpInterpSpeed=5
    minDownInterpSpeed=2
    minGustDuration=0.2
    minGustStrength=3
    minTimeBetweenGusts=1
    minUpInterpSpeed=2
    windTurbulenceScalarA=0.1
    windTurbulenceScalarB=0.8
    windAmplitude=<20 65 42>
    windFrequency=<5 6 8.6>
    windOffset=<50 -500 -400>
   }
   riding_Moving {
    maxDownInterpSpeed=5
    maxGustDuration=2
    maxGustStrength=7
    maxTimeBetweenGusts=7
    maxUpInterpSpeed=5
    minDownInterpSpeed=2
    minGustDuration=0.5
    minGustStrength=3
    minTimeBetweenGusts=0
    minUpInterpSpeed=2
    windTurbulenceScalarA=0.1
    windTurbulenceScalarB=1
    windAmplitude=<0 0 0>
    windFrequency=<0 0 0>
    windOffset=<0 0 0>
   }
   riding_Moving_Targeting {
    maxDownInterpSpeed=5
    maxGustDuration=2
    maxGustStrength=7
    maxTimeBetweenGusts=7
    maxUpInterpSpeed=5
    minDownInterpSpeed=2
    minGustDuration=0.5
    minGustStrength=3
    minTimeBetweenGusts=0
    minUpInterpSpeed=2
    windTurbulenceScalarA=0.1
    windTurbulenceScalarB=1
    windAmplitude=<0 0 0>
    windFrequency=<0 0 0>
    windOffset=<0 0 0>
   }
   riding_Standing {
    maxDownInterpSpeed=5
    maxGustDuration=2
    maxGustStrength=7
    maxTimeBetweenGusts=7
    maxUpInterpSpeed=5
    minDownInterpSpeed=2
    minGustDuration=0.5
    minGustStrength=3
    minTimeBetweenGusts=0
    minUpInterpSpeed=2
    windTurbulenceScalarA=0.1
    windTurbulenceScalarB=1
    windAmplitude=<0 0 0>
    windFrequency=<0 0 0>
    windOffset=<0 0 0>
   }
   skydive_DivingDown {
    maxDownInterpSpeed=5
    maxGustDuration=2
    maxGustStrength=7
    maxTimeBetweenGusts=7
    maxUpInterpSpeed=5
    minDownInterpSpeed=2
    minGustDuration=0.5
    minGustStrength=3
    minTimeBetweenGusts=0
    minUpInterpSpeed=2
    windTurbulenceScalarA=0.1
    windTurbulenceScalarB=1
    windAmplitude=<0 0 0>
    windFrequency=<0 0 0>
    windOffset=<0 0 0>
   }
   skydive_FloatingDown {
    maxDownInterpSpeed=5
    maxGustDuration=2
    maxGustStrength=7
    maxTimeBetweenGusts=7
    maxUpInterpSpeed=5
    minDownInterpSpeed=2
    minGustDuration=0.5
    minGustStrength=3
    minTimeBetweenGusts=0
    minUpInterpSpeed=2
    windTurbulenceScalarA=0.1
    windTurbulenceScalarB=1
    windAmplitude=<0 0 0>
    windFrequency=<0 0 0>
    windOffset=<0 0 0>
   }
   skydive_GoingUp {
    maxDownInterpSpeed=5
    maxGustDuration=2
    maxGustStrength=7
    maxTimeBetweenGusts=7
    maxUpInterpSpeed=5
    minDownInterpSpeed=2
    minGustDuration=0.5
    minGustStrength=3
    minTimeBetweenGusts=0
    minUpInterpSpeed=2
    windTurbulenceScalarA=0.1
    windTurbulenceScalarB=1
    windAmplitude=<0 0 0>
    windFrequency=<0 0 0>
    windOffset=<0 0 0>
   }
   skydive_Parachuting {
    maxDownInterpSpeed=5
    maxGustDuration=2
    maxGustStrength=7
    maxTimeBetweenGusts=7
    maxUpInterpSpeed=5
    minDownInterpSpeed=2
    minGustDuration=0.5
    minGustStrength=3
    minTimeBetweenGusts=0
    minUpInterpSpeed=2
    windTurbulenceScalarA=0.1
    windTurbulenceScalarB=1
    windAmplitude=<0 0 0>
    windFrequency=<0 0 0>
    windOffset=<0 0 0>
   }
   skydive_Surfing {
    maxDownInterpSpeed=5
    maxGustDuration=2
    maxGustStrength=7
    maxTimeBetweenGusts=7
    maxUpInterpSpeed=5
    minDownInterpSpeed=2
    minGustDuration=0.5
    minGustStrength=3
    minTimeBetweenGusts=0
    minUpInterpSpeed=2
    windTurbulenceScalarA=0.1
    windTurbulenceScalarB=1
    windAmplitude=<0 0 0>
    windFrequency=<0 0 0>
    windOffset=<0 0 0>
   }
   swimming {
    maxDownInterpSpeed=5
    maxGustDuration=2
    maxGustStrength=7
    maxTimeBetweenGusts=7
    maxUpInterpSpeed=5
    minDownInterpSpeed=2
    minGustDuration=0.5
    minGustStrength=3
    minTimeBetweenGusts=0
    minUpInterpSpeed=2
    windTurbulenceScalarA=0.1
    windTurbulenceScalarB=1
    windAmplitude=<0 0 0>
    windFrequency=<0 0 0>
    windOffset=<0 0 0>
   }
   tactical_Sprint {
    maxDownInterpSpeed=5
    maxGustDuration=2
    maxGustStrength=7
    maxTimeBetweenGusts=7
    maxUpInterpSpeed=5
    minDownInterpSpeed=2
    minGustDuration=0.5
    minGustStrength=3
    minTimeBetweenGusts=0
    minUpInterpSpeed=2
    windTurbulenceScalarA=0.1
    windTurbulenceScalarB=1
    windAmplitude=<0 0 0>
    windFrequency=<0 0 0>
    windOffset=<0 0 0>
   }
   ziplining {
    maxDownInterpSpeed=5
    maxGustDuration=2
    maxGustStrength=7
    maxTimeBetweenGusts=7
    maxUpInterpSpeed=5
    minDownInterpSpeed=2
    minGustDuration=0.5
    minGustStrength=3
    minTimeBetweenGusts=0
    minUpInterpSpeed=2
    windTurbulenceScalarA=0.1
    windTurbulenceScalarB=1
    windAmplitude=<0 0 0>
    windFrequency=<0 0 0>
    windOffset=<0 0 0>
   }
  }
 }
]
